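# Animation Blueprint for the FermiumStraw backpack, adding physics motion to its straps, bags and bottles.
# Chained AnimDynamics nodes simulate two-segment straps, planar-limited at the tips, and cache that pose for bag and bottle chains.
# The output selects an actor-space, linear-spring chain in the front end, straps alone while skydiving, and a component-space chain otherwise.
# The event graph sets simulation-space, velocity-scale and gravity variables from front-end, zipline and skydive-float states, though no visible node reads them.
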AnimBlueprint /BRCosmetics/Accessories/FORT_Backpacks/Backpack_F_MED_FermiumStraw/Meshes/F_MED_FermiumStraw_Pack_AnimBP
target_skeleton=&/F_MED_FermiumStraw_Pack_Skeleton
graphs [
 AnimGraph {nodes [
   n1 = Anim Dynamics - Bone: dyn_bag_lf!(
    ComponentPose=Local To Component(LocalPose=Use cached pose 'Straps_CACHE'())
    Alpha=1
    ComponentLinearAccScale:Vector="1,1,1"
    ComponentLinearVelScale:Vector="1,1,1"
    ExternalForce="0,0,0"
   ) -> (Pose) {
    angularDampingOverride=1
    angularSpringConstant=11000
    bAngularSpring=true
    bOverrideAngularDamping=true
    bUseGravityOverride=true
    gravityScale=0.35
    linearDampingOverride=1
    alphaBoolBlend {blendOption=Linear}
    boundBone {boneName=dyn_bag_lf}
    componentAppliedLinearAccClamp=<1000 1000 1000>
    componentLinearAccScale=<1 1 1>
    componentLinearVelScale=<1 1 1>
    gravityOverride {z=-1010}
    physicsBodyDefinitions [{
      boundBone {boneName=dyn_bag_lf}
      boxExtents=<60 60 60>
      constraintSetup {angularLimitsMax=<0 3 0>, angularLimitsMin=<0 -2 0>, angularTarget=<1 0 0>, linearAxesMax=<0.5 0 0>, linearAxesMin=<-0.1 0 0>}
      localJointOffset=<1.1 0 0>
     }]
    simSpaceSettings {maxAngularAcceleration=20, maxAngularVelocity=4, simSpaceAngularAlpha=1}
   }
   n2 = Anim Dynamics - Bone: dyn_bag_bottle_lf!(ComponentPose=@n1, Alpha=1, ComponentLinearAccScale:Vector="1,1,1", ComponentLinearVelScale:Vector="1,1,1", ExternalForce="0,0,0") -> (Pose) {
    angularDampingOverride=1
    angularSpringConstant=11000
    bAngularSpring=true
    bOverrideAngularDamping=true
    bUseGravityOverride=true
    gravityScale=0.35
    linearDampingOverride=1
    alphaBoolBlend {blendOption=Linear}
    boundBone {boneName=dyn_bag_bottle_lf}
    componentAppliedLinearAccClamp=<1000 1000 1000>
    componentLinearAccScale=<1 1 1>
    componentLinearVelScale=<1 1 1>
    gravityOverride {z=-1000}
    physicsBodyDefinitions [{
      boundBone {boneName=dyn_bag_bottle_lf}
      boxExtents=<80 80 80>
      constraintSetup {angularLimitsMax=<0 4 0>, angularLimitsMin=<0 -2 0>, angularTarget=<1 0 0>, linearAxesMax=<0.8 0 0>, linearAxesMin=<-0.2 0 0>}
      localJointOffset=<2.5 0 0>
     }]
    simSpaceSettings {maxAngularAcceleration=24, maxAngularVelocity=4, simSpaceAngularAlpha=1}
   }
   n3 = Anim Dynamics - Bone: dyn_bag_bottle_rt!(ComponentPose=@n2, Alpha=1, ComponentLinearAccScale:Vector="1,1,1", ComponentLinearVelScale:Vector="1,1,1", ExternalForce="0,0,0") -> (Pose) {
    angularDampingOverride=1
    angularSpringConstant=11000
    bAngularSpring=true
    bOverrideAngularDamping=true
    bUseGravityOverride=true
    gravityScale=0.35
    linearDampingOverride=1
    alphaBoolBlend {blendOption=Linear}
    boundBone {boneName=dyn_bag_bottle_rt}
    componentAppliedLinearAccClamp=<1000 1000 1000>
    componentLinearAccScale=<1 1 1>
    componentLinearVelScale=<1 1 1>
    gravityOverride {z=-1000}
    physicsBodyDefinitions [{
      boundBone {boneName=dyn_bag_bottle_rt}
      boxExtents=<75 75 75>
      constraintSetup {angularLimitsMax=<0 3 0>, angularLimitsMin=<0 -3 0>, angularTarget=<1 0 0>, linearAxesMax=<0.9 0 0>, linearAxesMin=<-0.2 0 0>}
      localJointOffset=<2.2 0 0>
     }]
    simSpaceSettings {maxAngularAcceleration=22, maxAngularVelocity=4, simSpaceAngularAlpha=1}
   }
   n4 = Anim Dynamics - Bone: dyn_bag_rt!(ComponentPose=@n3, Alpha=1, ComponentLinearAccScale:Vector="1,1,1", ComponentLinearVelScale:Vector="1,1,1", ExternalForce="0,0,0") -> (Pose) {
    angularDampingOverride=1
    angularSpringConstant=11000
    bAngularSpring=true
    bOverrideAngularDamping=true
    bUseGravityOverride=true
    gravityScale=0.35
    linearDampingOverride=1
    alphaBoolBlend {blendOption=Linear}
    boundBone {boneName=dyn_bag_rt}
    componentAppliedLinearAccClamp=<1000 1000 1000>
    componentLinearAccScale=<1 1 1>
    componentLinearVelScale=<1 1 1>
    gravityOverride {z=-1010}
    physicsBodyDefinitions [{
      boundBone {boneName=dyn_bag_rt}
      boxExtents=<65 65 65>
      constraintSetup {angularLimitsMax=<0 3 0>, angularLimitsMin=<0 -1 0>, angularTarget=<1 0 0>, linearAxesMax=<0.8 0 0>, linearAxesMin=<-0.2 0 0>}
      localJointOffset=<1.5 0 0>
     }]
    simSpaceSettings {maxAngularAcceleration=18, maxAngularVelocity=4, simSpaceAngularAlpha=1}
   }
   n5 = Component To Local(ComponentPose=@n4)
   n6 = Use cached pose 'Straps_CACHE'()
   n7 = Get bIsSkyDiving()
   n8 = Anim Dynamics - Bone: dyn_bag_lf(
    ComponentPose=Local To Component(LocalPose=Use cached pose 'Straps_CACHE'())
    Alpha=1
    ComponentLinearAccScale="1,1,1"
    ComponentLinearVelScale="1,1,1"
    ExternalForce="0,0,0"
   ) {
    angularDampingOverride=1
    angularSpringConstant=11000
    bAngularSpring=true
    bLinearSpring=true
    bOverrideAngularDamping=true
    bOverrideLinearDamping=true
    bUseGravityOverride=true
    gravityScale=0.35
    linearDampingOverride=1
    linearSpringConstant=20
    simulationSpace=Actor
    alphaBoolBlend {blendOption=Linear}
    boundBone {boneName=dyn_bag_lf}
    componentAppliedLinearAccClamp=<500 500 500>
    componentLinearAccScale=<1 1 1>
    componentLinearVelScale=<1 1 1>
    gravityOverride {z=-1010}
    physicsBodyDefinitions [{
      boundBone {boneName=dyn_bag_lf}
      boxExtents=<60 60 60>
      constraintSetup {angularLimitsMax=<0 3 0>, angularLimitsMin=<0 -2 0>, angularTarget=<1 0 0>, linearAxesMax=<0.4 0 0>}
      localJointOffset=<2 0 0>
     }]
    simSpaceSettings {maxAngularAcceleration=20, maxAngularVelocity=4, simSpaceAngularAlpha=1}
   }
   n9 = Anim Dynamics - Bone: dyn_bag_bottle_lf(ComponentPose=@n8, Alpha=1, ComponentLinearAccScale="1,1,1", ComponentLinearVelScale="1,1,1", ExternalForce="0,0,0") {
    angularDampingOverride=1
    angularSpringConstant=11000
    bAngularSpring=true
    bLinearSpring=true
    bOverrideAngularDamping=true
    bOverrideLinearDamping=true
    bUseGravityOverride=true
    gravityScale=0.35
    linearDampingOverride=1
    linearSpringConstant=20
    simulationSpace=Actor
    alphaBoolBlend {blendOption=Linear}
    boundBone {boneName=dyn_bag_bottle_lf}
    componentAppliedLinearAccClamp=<500 500 500>
    componentLinearAccScale=<1 1 1>
    componentLinearVelScale=<1 1 1>
    gravityOverride {z=-1000}
    physicsBodyDefinitions [{
      boundBone {boneName=dyn_bag_bottle_lf}
      boxExtents=<80 80 80>
      constraintSetup {angularLimitsMax=<0 4 0>, angularLimitsMin=<0 -2 0>, angularTarget=<1 0 0>, linearAxesMax=<0.3 0 0>}
      localJointOffset=<1 0 0>
     }]
    simSpaceSettings {maxAngularAcceleration=24, maxAngularVelocity=4, simSpaceAngularAlpha=1}
   }
   n10 = Anim Dynamics - Bone: dyn_bag_bottle_rt(ComponentPose=@n9, Alpha=1, ComponentLinearAccScale="1,1,1", ComponentLinearVelScale="1,1,1", ExternalForce="0,0,0") {
    angularDampingOverride=1
    angularSpringConstant=11000
    bAngularSpring=true
    bLinearSpring=true
    bOverrideAngularDamping=true
    bOverrideLinearDamping=true
    bUseGravityOverride=true
    gravityScale=0.35
    linearDampingOverride=1
    linearSpringConstant=20
    simulationSpace=Actor
    alphaBoolBlend {blendOption=Linear}
    boundBone {boneName=dyn_bag_bottle_rt}
    componentAppliedLinearAccClamp=<500 500 500>
    componentLinearAccScale=<1 1 1>
    componentLinearVelScale=<1 1 1>
    gravityOverride {z=-1000}
    physicsBodyDefinitions [{
      boundBone {boneName=dyn_bag_bottle_rt}
      boxExtents=<75 75 75>
      constraintSetup {angularLimitsMax=<0 3 0>, angularLimitsMin=<0 -3 0>, angularTarget=<1 0 0>, linearAxesMax=<0.3 0 0>}
      localJointOffset=<2 0 0>
     }]
    simSpaceSettings {maxAngularAcceleration=22, maxAngularVelocity=4, simSpaceAngularAlpha=1}
   }
   n11 = Anim Dynamics - Bone: dyn_bag_rt(ComponentPose=@n10, Alpha=1, ComponentLinearAccScale="1,1,1", ComponentLinearVelScale="1,1,1", ExternalForce="0,0,0") {
    angularDampingOverride=1
    angularSpringConstant=11000
    bAngularSpring=true
    bLinearSpring=true
    bOverrideAngularDamping=true
    bOverrideLinearDamping=true
    bUseGravityOverride=true
    gravityScale=0.35
    linearDampingOverride=1
    linearSpringConstant=20
    simulationSpace=Actor
    alphaBoolBlend {blendOption=Linear}
    boundBone {boneName=dyn_bag_rt}
    componentAppliedLinearAccClamp=<500 500 500>
    componentLinearAccScale=<1 1 1>
    componentLinearVelScale=<1 1 1>
    gravityOverride {z=-1010}
    physicsBodyDefinitions [{
      boundBone {boneName=dyn_bag_rt}
      boxExtents=<65 65 65>
      constraintSetup {angularLimitsMax=<0 3 0>, angularLimitsMin=<0 -1 0>, angularTarget=<1 0 0>, linearAxesMax=<0.4 0 0>}
      localJointOffset=<2 0 0>
     }]
    simSpaceSettings {maxAngularAcceleration=18, maxAngularVelocity=4, simSpaceAngularAlpha=1}
   }
   Output Pose(Result=Blend Poses by bool(
     BlendPose_0=Component To Local(ComponentPose=@n11)
     BlendPose_1=Blend Poses by bool(BlendPose_0=@n6, BlendPose_1=@n5, bActiveValue=@n7)
     bActiveValue=Get bIsInFrontEnd()
    )) {layerGroup=None}
   n12 = Anim Dynamics - Bone: dyn_strap_lf_01!(
    SimulationSpace:AnimPhysSimSpaceType Enum=World
    ComponentPose=Local To Component(LocalPose=Input Pose())
    Alpha=1
    GravityOverride:Vector="0,0,-980"
    ComponentLinearAccScale:Vector="1,1,1"
    ComponentLinearVelScale:Vector="1,1,1"
    ExternalForce="0,0,0"
   ) -> (Pose) {
    angularDampingOverride=1
    angularSpringConstant=11000
    bAngularSpring=true
    bOverrideAngularDamping=true
    bUseGravityOverride=true
    gravityScale=0.35
    simulationSpace=World
    alphaBoolBlend {blendOption=Linear}
    boundBone {boneName=dyn_strap_lf_01}
    componentAppliedLinearAccClamp=<1000 1000 1000>
    componentLinearAccScale=<1 1 1>
    componentLinearVelScale=<1 1 1>
    gravityOverride {z=-980}
    physicsBodyDefinitions [{
      boundBone {boneName=dyn_strap_lf_01}
      boxExtents=<25 30 25>
      constraintSetup {angularLimitsMax=<5 45 25>, angularLimitsMin=<-5 -45 -35>, angularTarget=<1 0 0>}
      localJointOffset=<7 0 0>
     }]
    simSpaceSettings {maxAngularAcceleration=35, maxAngularVelocity=8, simSpaceAngularAlpha=1}
   }
   n13 = Anim Dynamics - Bone: dyn_strap_lf_02!(
    SimulationSpace:AnimPhysSimSpaceType Enum=World
    ComponentPose=@n12
    Alpha=1
    GravityOverride:Vector="0,0,-980"
    ComponentLinearAccScale:Vector="1,1,1"
    ComponentLinearVelScale:Vector="1,1,1"
    ExternalForce="0,0,0"
   ) -> (Pose) {
    angularDampingOverride=1
    angularSpringConstant=11000
    bAngularSpring=true
    bOverrideAngularDamping=true
    bUseGravityOverride=true
    gravityScale=0.35
    simulationSpace=World
    alphaBoolBlend {blendOption=Linear}
    boundBone {boneName=dyn_strap_lf_02}
    componentAppliedLinearAccClamp=<1000 1000 1000>
    componentLinearAccScale=<1 1 1>
    componentLinearVelScale=<1 1 1>
    gravityOverride {z=-980}
    physicsBodyDefinitions [{
      collisionType=CustomSphere
      sphereCollisionRadius=2
      boundBone {boneName=dyn_strap_lf_02}
      boxExtents=<25 30 25>
      constraintSetup {angularLimitsMax=<10 80 35>, angularLimitsMin=<-10 -80 -44>, angularTarget=<1 0 0>}
      localJointOffset=<7 0 0>
     }]
    planarLimits [{drivingBone {boneName=pack}, planeTransform {location=<0 -20 -14>, rotation=<rot 0 0 180>}}]
    simSpaceSettings {maxAngularAcceleration=35, maxAngularVelocity=8, simSpaceAngularAlpha=1}
   }
   n14 = Anim Dynamics - Bone: dyn_strap_rt_01!(
    SimulationSpace:AnimPhysSimSpaceType Enum=Component
    ComponentPose=@n13
    Alpha=1
    GravityOverride:Vector="0,0,-980"
    ComponentLinearAccScale:Vector="1,1,1"
    ComponentLinearVelScale:Vector="1,1,1"
    ExternalForce="0,0,0"
   ) -> (Pose) {
    angularDampingOverride=1
    angularSpringConstant=11000
    bAngularSpring=true
    bOverrideAngularDamping=true
    bUseGravityOverride=true
    gravityScale=0.35
    alphaBoolBlend {blendOption=Linear}
    boundBone {boneName=dyn_strap_rt_01}
    componentAppliedLinearAccClamp=<1000 1000 1000>
    componentLinearAccScale=<1 1 1>
    componentLinearVelScale=<1 1 1>
    gravityOverride {z=-980}
    physicsBodyDefinitions [{
      boundBone {boneName=dyn_strap_rt_01}
      boxExtents=<30 25 30>
      constraintSetup {angularLimitsMax=<5 45 30>, angularLimitsMin=<-5 -45 -30>, angularTarget=<1 0 0>}
      localJointOffset=<8 0 0>
     }]
    simSpaceSettings {maxAngularAcceleration=35, maxAngularVelocity=8, simSpaceAngularAlpha=1}
   }
   n15 = Anim Dynamics - Bone: dyn_strap_rt_02!(
    SimulationSpace:AnimPhysSimSpaceType Enum=Component
    ComponentPose=@n14
    Alpha=1
    GravityOverride:Vector="0,0,-980"
    ComponentLinearAccScale:Vector="1,1,1"
    ComponentLinearVelScale:Vector="1,1,1"
    ExternalForce="0,0,0"
   ) -> (Pose) {
    angularDampingOverride=1
    angularSpringConstant=11000
    bAngularSpring=true
    bOverrideAngularDamping=true
    bUseGravityOverride=true
    gravityScale=0.35
    alphaBoolBlend {blendOption=Linear}
    boundBone {boneName=dyn_strap_rt_02}
    componentAppliedLinearAccClamp=<1000 1000 1000>
    componentLinearAccScale=<1 1 1>
    componentLinearVelScale=<1 1 1>
    gravityOverride {z=-980}
    physicsBodyDefinitions [{
      collisionType=CustomSphere
      sphereCollisionRadius=2
      boundBone {boneName=dyn_strap_rt_02}
      boxExtents=<30 25 30>
      constraintSetup {angularLimitsMax=<10 89 44>, angularLimitsMin=<-10 -89 -44>, angularTarget=<1 0 0>}
      localJointOffset=<8 0 0>
     }]
    planarLimits [{drivingBone {boneName=pack}, planeTransform {location=<0 -20 -14>, rotation=<rot 0 0 180>}}]
    simSpaceSettings {maxAngularAcceleration=35, maxAngularVelocity=8, simSpaceAngularAlpha=1}
   }
   Save cached pose 'Straps_CACHE'(Pose=Component To Local(ComponentPose=@n15)) {cacheName=Straps_CACHE}
  ]}
 EventGraph {nodes [
   Event BlueprintUpdateAnimation()
   > n1 = Sequence!(execute) -> (then_0, then_1, then_2:Exec)
   n2 = GetOwnerAnimationInterface!(self:Custom Character Part Anim Instance Object Reference) -> (ReturnValue:Fort Player Animation Interface Object Reference)
   > Is Valid(InputObject=@n2)
   > n3 = Branch(Condition=Break Fort Anim Input Zipline!(FortAnimInput_Zipline:"Fort Anim Input Zipline Structure (by ref)"=GetZiplineInput!(self:Fort Player Animation Interface Object Reference=@n2) -> (ReturnValue:"Fort Anim Input Zipline Structure (by ref)")) -> (
     bIsZiplining:bool=~
     bIsOnAscender:bool=~
     bShouldPlayPivotTransition:bool=~
     LeanYaw:float=~
     PivotBlendDelayRemaining:float=~
     PivotBlendDelay:float=~
     PivotCardinalDirection:EFortCardinalDirection Enum
     WorldVelocityLastTick:Vector=~
    ) {is=title K2Node_BreakStruct})
   > Set Componen Linear Vel Scale(Componen Linear Vel Scale:Vector="0, 0, 0") -> (Output_Get:Vector)
   Set Componen Linear Vel Scale(execute=@n3.else, Componen Linear Vel Scale="0.5,0.5,0.5")
   n4 = Branch(execute=@n1.then_1, Condition=Get bIsInFrontEnd())
   > Set AnimDynamic_SimulationSpace(AnimDynamic_SimulationSpace:Byte=1) -> (Output_Get:Byte=0)
   Set AnimDynamic_SimulationSpace(execute=@n4.else, AnimDynamic_SimulationSpace=0) -> (Output_Get=0)
   n5 = Get bIsJumping()
   n6 = Branch(execute=@n1.then_2, Condition=AND Boolean(A=NOT Boolean(A=@n5), B=Get bIsSkydiveFloating()))
   > Set Gravity Override(Gravity Override="0,0,980")
   Set Gravity Override(execute=@n6.else, Gravity Override="0,0,-980")
   TryGetPawnOwner()
  ]}
]
variables [
 AnimDynamic_SimulationSpace {type=integer}
 Componen Linear Vel Scale {type=Vector, default=<0.5 0.5 0.5>}
 Gravity Override {type=Vector}
]
